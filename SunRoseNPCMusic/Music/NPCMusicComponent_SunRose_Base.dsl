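; Base NPC music component for SunRose boss encounters, extending a parent component with phase-based music helpers.
; Entering the audio volume sends the current phase to the music manager as a starting-phase parameter; exiting stops the transition mix.
; Inside the volume, a modulation control bus mix can activate for transitions, with a timer to deactivate it.
; Helpers resize the volume and override group music attenuation, while the threatened-state check has no action.
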(event EventBeginPlay
  ;[0 0 0] Removes sunrose component for cerberus and ares minions/altars to prevent issues when Boss spawns
  (|CallParentFunction))

(event EventEndPlay
  (params (EndPlayReason :type EEndPlayReason%20Enum :default "Destroyed"))
  (|CallParentFunction EndPlayReason))

(fn ChangeAGVRadius
  (params (InRadius :type Float%20%28single-precision%29))
  (bind (_agv _bisvalid) (Public|GetAGV))
  (if _bisvalid
    (Control|SetRadius (Variables|/FortNPCMusic/Audio/AGV_NPCMusicVolume.AGV_NPCMusicVolume_C|GetAudioShapeSphere _agv) InRadius :!class /Script/AudioShapes.AudioShapeSphereComponent)))

(fn CheckNPCState
  (bind _asbp_player_pawn_non_participant_boss (Utilities|Casting|CastToBP_PlayerPawn_NonParticipant_Boss (Components|GetOwner)))
  (if (Utilities|Enum|Equal(Enum) (Variables|/Game/Athena/AI/NPCs/Base/BP_PlayerPawn_NonParticipant.BP_PlayerPawn_NonParticipant_C|GetAlertLevel _asbp_player_pawn_non_participant_boss) "Threatened")
    _))

(fn OnPlayerEnterAudioVolume
  (Class|NPCMusicManagerComponent|SetMusicFloatParam (Variables|Private|GetMusicManager) self "NPC.Music.StartingPhase" (Math|Conversions|ToFloat(Integer) (Utilities|Select :Index (Variables|Default|GetCurrentPhaseEnum) :NewEnumerator0 0 :NewEnumerator1 1 :NewEnumerator2 2 :NewEnumerator3 3)))
  (Variables|Default|SetbIsInAGV true))

(fn OnPlayerExitAudioVolume
  (Variables|Default|SetbIsInAGV false)
  (CallFunction|StopTransitionModulationMix))

(fn SetAltarAttenuationOverride
  (params (Attenuation%20Settings :type Sound%20Attenuation%20Object%20Reference))
  (bind _self self)
  (bind _musicmanager (Variables|Private|GetMusicManager))
  (bind (_isvalid _groupcomponent) (Class|NPCMusicManagerComponent|GetGroupComponent _musicmanager _self))
  (if _isvalid
    (Variables|/Script/Engine.AudioComponent|SetAttenuationSettings Attenuation%20Settings (Variables|/FortNPCMusic/Audio/NPCMusicActorGroupComponent.NPCMusicActorGroupComponent_C|GetMusic _groupcomponent))
    (for _ (Variables|/FortNPCMusic/Audio/NPCMusicActorGroupComponent.NPCMusicActorGroupComponent_C|GetComponents _groupcomponent)
      _)))

(fn SetTransitionModulationMixTimer
  (params (Duration :type Float%20%28single-precision%29 :default 0.1))
  (if (Variables|Default|GetbIsInAGV)
    (Audio|Modulation|ActivateControlBusMix "/SunRoseNPCMusic/Music/CBM_SunRoseMusicEvent.CBM_SunRoseMusicEvent")
    (if (not (Utilities|Time|IsValidTimerHandle (Variables|Default|GetTransitionMixTimer)))
      (bind _returnvalue (Utilities|Time|SetTimerbyEvent (EventDispatchers|CreateEvent :!event StopTransitionModulationMix) Duration))
      (Variables|Default|SetTransitionMixTimer _returnvalue))))

(fn SetupPawnBindings
  (|CallParentFunction))

(fn StopTransitionModulationMix
  (bind _transitionmixtimer (Variables|Default|GetTransitionMixTimer))
  (Audio|Modulation|DeactivateControlBusMix "/SunRoseNPCMusic/Music/CBM_SunRoseMusicEvent.CBM_SunRoseMusicEvent")
  (if (Utilities|Time|IsValidTimerHandle _transitionmixtimer)
    (Utilities|Time|ClearandInvalidateTimerbyHandle _transitionmixtimer)))
